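# Niagara system for the Spacetear vehicle boost cosmetic, trailing layered ribbons and small sprite particles from the vehicle's boost sockets.
# The ribbon emitter runs seven renderers, with curve-driven glow widths partly scaled by color variant and per-variant painted color overrides.
# The sprite emitter partitions particles across sockets and fades them under drag.
# Game-mode switches tune spawn rate, velocity and emissive color, and both emitters inherit decaying source movement.
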NiagaraSystem /VehicleCosmetics/Boosts/Boost_Spacetear/FX/NS_Boost_Spacetear
dependencies {
 usedDataInterfaces [&NiagaraDataInterfaceArrayFloat, &NiagaraDataInterfaceSkeletalMesh]
 usedDynamicInputs []
 usedModules [
  &self:SystemSpawnScript.SystemScriptSource.SystemScriptGraph.NiagaraNodeAssignment_0.SetVariables_#1
  &SystemState
 ]
}
emitters {
 NE_VC_Boost_Ribbon_Parent001 {
  emitterUpdateScript [
   &EmitterState(Inactive Response=::NewEnumerator1, Life Cycle Mode=::NewEnumerator1, Loop Duration=1, UseLoopDelay=false)
   &VC_FilteredSocketCount()
   &VC_RateQualityHandler(Default Rate:NiagaraFloat=&Multiply_Float(
     A=50
     B=/VehicleCosmetics/FX/Niagara/Modules/VC_Mask_Floats(Coro Value:NiagaraFloat=1, FrontEnd Value:NiagaraFloat=0.75, FN Value:NiagaraFloat=0.5)
    ))
   &VC_SpawnRibbon_v1(
    Number of Ribbons=$Emitter.PortCount
    SpawnRate=&MergeHelper_FloatSwitch(0=80, 1=20, 2=20, 3=80, Index=$System.GameModeID, Num=::NewEnumerator2)
   )
   /SetVariables(
    Emitter.Panner Speed:NiagaraFloat=$User.Speed Ribbon
    Emitter.RGB Emiss Stars:LinearColor=&MergeHelper_ColorSwitch(0=<rgba 3 0 0 0>, 1=<rgba 12 0 0 0>, 2=<rgba 12 0 0 0>, 3=<rgba 3 0 0 0>, Index=$User.GameModeID, Num=::NewEnumerator2)
   ) {isSetParameters=true}
  ]
  particleSpawnScript [
   &VC_PaintedVariantsOverride(
    Black=<rgba 0.2656 0.2656 0.2656 1>
    BurntSienna=<rgba 5 1.0578 0 1>
    Cobalt=<rgba 1.9878 2.0436 5 1>
    Crimson=<rgba 1 0.1128 0.1128 1>
    DefaultColor=<rgba 0 0.0093 2 1>
    ForestGreen=<rgba 0 1 0.0505 1>
    Grey=<rgba 0.2031 0.2031 0.2031 0>
    Lime=<rgba 0.5 1 0 1>
    Orange=<rgba 1 0 0.0372 1>
    Pink=<rgba 2 0.2396 1.3598 1>
    Purple=<rgba 0.2789 0 0.6795 0>
    R1 Black:LinearColor=<rgba 0.1163 0.1163 0.1163 1>
    R1 BurntSienna:LinearColor=<rgba 0.474 0.2576 0.0312 1>
    R1 Cobalt:LinearColor=<rgba 0 0.0185 1 1>
    R1 Crimson:LinearColor=<rgba 3 0 0 1>
    R1 DefaultColor=<rgba 1.9265 0.7664 2 1>
    R1 ForestGreen:LinearColor=<rgba 0 1 0.0505 1>
    R1 Grey:LinearColor=<rgba 0.2031 0.2031 0.2031 0>
    R1 Lime:LinearColor=<rgba 1 2 0 1>
    R1 Orange:LinearColor=<rgba 2 0.3894 0 1>
    R1 Pink:LinearColor=<rgba 1.0537 0 2 1>
    R1 Purple:LinearColor=<rgba 0.2789 0 0.6795 0>
    R1 Saffron:LinearColor=<rgba 3 2.3 0 1>
    R1 SkyBlue:LinearColor=<rgba 0.2994 0 3 1>
    R1 TitaniumWhite:LinearColor=<rgba 2 1.0976 0 1>
    R2 BurntSienna:LinearColor=<rgba 0.03 0 0 1>
    R2 Cobalt:LinearColor=<rgba 0 0 0.03 1>
    R2 Crimson:LinearColor=<rgba 0.015 0 0.0003 1>
    R2 DefaultColor:LinearColor=<rgba 0 0 0 1>
    R2 ForestGreen:LinearColor=<rgba 0 0.01 0 1>
    R2 Grey:LinearColor=<rgba 0.2031 0.2031 0.2031 0>
    R2 Lime:LinearColor=<rgba 0.005 0.01 0 1>
    R2 Orange:LinearColor=<rgba 0 0 0 1>
    R2 Pink:LinearColor=<rgba 0.01 0.0012 0.0068 1>
    R2 Purple:LinearColor=<rgba 0.2789 0 0.6795 0>
    R2 Saffron:LinearColor=<rgba 0.01 0.0077 0 1>
    R2 SkyBlue:LinearColor=<rgba 0 0.0067 0.01 1>
    R2 TitaniumWhite:LinearColor=<rgba 0.01 0.01 0.01 1>
    R3 BurntSienna:LinearColor=<rgba 0.7969 0.2036 0.1026 1>
    R3 Cobalt:LinearColor=<rgba 0 0.0466 0.5503 1>
    R3 Crimson:LinearColor=<rgba 1 0 0 1>
    R3 DefaultColor:LinearColor=<rgba 0 0 0.658 1>
    R3 ForestGreen:LinearColor=<rgba 0 0.4219 0.0213 0.3507>
    R3 Grey:LinearColor=<rgba 0.2031 0.2031 0.2031 0>
    R3 Lime:LinearColor=<rgba 0.9313 1.5 0 1>
    R3 Orange:LinearColor=<rgba 2 0.6976 0 1>
    R3 Pink:LinearColor=<rgba 1 0.1545 0.8853 1>
    R3 Purple:LinearColor=<rgba 0.2789 0 0.6795 0>
    R3 Saffron:LinearColor=<rgba 1.2 1.12 0 1>
    R3 SkyBlue:LinearColor=<rgba 0 0.9981 1.5 1>
    R3 TitaniumWhite:LinearColor=<rgba 0.8385 0.8385 0.8385 1>
    R4 BurntSienna:LinearColor=<rgba 0.6302 0.2712 0.0973 0>
    R4 Cobalt:LinearColor=<rgba 0 0.0185 1 1>
    R4 Crimson:LinearColor=<rgba 1 0 0 1>
    R4 DefaultColor:LinearColor=<rgba 0.3333 0 1 1>
    R4 ForestGreen:LinearColor=<rgba 0 0.6927 0.035 1>
    R4 Grey:LinearColor=<rgba 0.2031 0.2031 0.2031 0>
    R4 Lime:LinearColor=<rgba 0.9604 1 0 1>
    R4 Pink:LinearColor=<rgba 0.3 0 0.2926 1>
    R4 Purple:LinearColor=<rgba 0.2789 0 0.6795 0>
    R4 SkyBlue:LinearColor=<rgba 0 0.6654 1 1>
    R4 TitaniumWhite:LinearColor=<rgba 0.8906 0.8906 0.8906 1>
    R5 Black:LinearColor=<rgba 1 1 1 0>
    R5 BurntSienna:LinearColor=<rgba 1 1 1 0>
    R5 Cobalt:LinearColor=<rgba 1 1 1 0>
    R5 Crimson:LinearColor=<rgba 1 1 1 0>
    R5 DefaultColor:LinearColor=<rgba 1 1 1 0>
    R5 ForestGreen:LinearColor=<rgba 1 1 1 0>
    R5 Gold:LinearColor=<rgba 1 1 1 0>
    R5 Grey:LinearColor=<rgba 1 1 1 0>
    R5 Lime:LinearColor=<rgba 1 1 1 0>
    R5 Orange:LinearColor=<rgba 1 1 1 0>
    R5 Pink:LinearColor=<rgba 1 1 1 0>
    R5 Purple:LinearColor=<rgba 1 1 1 0>
    R5 Saffron:LinearColor=<rgba 1 1 1 0>
    R5 SkyBlue:LinearColor=<rgba 1 1 1 0>
    R5 TitaniumWhite:LinearColor=<rgba 0.2544 0 2 1>
    Saffron=<rgba 2 0.9942 0 1>
    SkyBlue=<rgba 0 1.9962 3 1>
    TitaniumWhite=<rgba 2 2 2 1>
    Use R1 Painted Overrides=true
    Use R2 Painted Overrides=true
    Use R3 Painted Overrides=true
    Use R4 Painted Overrides=true
    Use R5 Painted Overrides=true
   )
   &VC_InitializeRibbon_v1(
    Color=$Particles.Color
    Enable Skeletal Position=true
    Lifetime=0.8
    Offset Amount:NiagaraFloat=&MergeHelper_FloatSwitch(0=0, 1=50, 2=50, 3=0, Index=$User.GameModeID, Num=::NewEnumerator2)
    Offset Axis:ENiagaraOrientationAxis=::XAxis
   )
   &AddVelocity(Velocity=&MergeHelper_VectorSwitch(0=<-1750 0 0>, 1=<-1300 0 0>, 2=<-1300 0 0>, 3=<-1750 0 0>, Index=$User.GameModeID, Num=::NewEnumerator2)) {module=AddVelocity001}
   /VehicleCosmetics/FX/VC_FXParents/Modules/DLR_ProductPreviews(GameModeID:NiagaraInt32=$System.GameModeID, bPreviewVehicleSpeed:NiagaraBool=$System.bPreviewVehicleSpeed)
  ]
  particleUpdateScript [
   &ParticleState()
   &InheritSourceMovement(
    Applied Position Scale=&VectorFromFloat(Value=&FloatFromCurve(FloatCurve=DataInterface() {propertyValues {Curve {
         defaultValue=1
         keys [
          {arriveTangent=-1.7885, interpMode=RCIM_Cubic, leaveTangent=-1.7885, tangentMode=RCTM_User, time=0.05, value=1}
          {interpMode=RCIM_Cubic, tangentMode=RCTM_User, time=1, value=0}
         ]
        }}}))
    ApplyForce=false
    ApplyPosition=true
    ApplyPositionFalloff=false
   )
   &SolveForcesAndVelocity(Clamp Velocity=false, Limit Acceleration=false)
   /SetVariables(
    Particles.Size RIbbon Glow:NiagaraFloat=&FloatFromCurve(
     FloatCurve=$User.Ribbon General Ribbon
     Scale Curve=&SelectFloatFromArray(
      Direct Array Index=$User.ColorVariantID
      Float Selection Array=DataInterface() {propertyValues {FloatData [40,40,40,50,25,40,40,40,40,45,40,40,40,40,40]}}
     )
    )
    Particles.Size Ribbon Secon Glow:NiagaraFloat=&FloatFromCurve(FloatCurve=$User.Ribbon General Ribbon, Scale Curve=40)
    Particles.Size Ribbon Tree Glow:NiagaraFloat=&FloatFromCurve(
     FloatCurve=DataInterface() {propertyValues {Curve {keys [
         {arriveTangent=-0.0968, interpMode=RCIM_Cubic, leaveTangent=-0.0968, tangentMode=RCTM_User, time=0, value=1}
         {arriveTangent=-0.9579, interpMode=RCIM_Cubic, leaveTangent=-0.9579, tangentMode=RCTM_User, time=0.9, value=0}
        ]}}}
     Scale Curve=&SelectFloatFromArray(
      Direct Array Index=$User.ColorVariantID
      Float Selection Array=DataInterface() {propertyValues {FloatData [190,190,190,150,190,130,140,130,140,130,180,130,130,130,130]}}
     )
    )
    Particles.Size Stars 01:NiagaraFloat=&FloatFromCurve(FloatCurve=$User.Ribbon General Ribbon, Scale Curve=23)
   ) {isSetParameters=true}
   &VC_UpdateRibbon_v1(Enable Skeletal Position=true)
  ]
  rendererClasses [&NiagaraRibbonRendererProperties]
  renderers [
   {rendererClass=&NiagaraRibbonRendererProperties}
   {rendererClass=&NiagaraRibbonRendererProperties, rendererIndex=1}
   {rendererClass=&NiagaraRibbonRendererProperties, rendererIndex=2}
   {rendererClass=&NiagaraRibbonRendererProperties, rendererIndex=3}
   {rendererClass=&NiagaraRibbonRendererProperties, rendererIndex=4}
   {rendererClass=&NiagaraRibbonRendererProperties, rendererIndex=5}
   {rendererClass=&NiagaraRibbonRendererProperties, rendererIndex=6}
  ]
 }
 NE_VC_Boost_Sprite_Parent {
  emitterUpdateScript [
   &EmitterState()
   &VC_FilteredSocketCount()
   &SpawnRate(
    SpawnRate=&self:ScratchDynamicInput(RateBR:NiagaraFloat=250, RateFR:NiagaraFloat=20, RateIG:NiagaraFloat=200)
    Use Spawn Probability=false
   )
   &self:VC_PlayMode(
    BR_Color:LinearColor=&MakeCustomLinearColorFromBool(Bool=$User.bIsPainted, True Color=$User.PaintedColor, False Color=<rgba 0.3 0.2 0.45 1>)
    BR_Lifetime:NiagaraFloat=&RandomRangeFloat(Minimum=0.5, Maximum=0.6)
    BR_Velocity:Vector3f=<0 0 0>
    FR_Color:LinearColor=&MakeCustomLinearColorFromBool(Bool=$User.bIsPainted, True Color=$User.PaintedColor, False Color=<rgba 0.3 0.2 0.45 1>)
    FR_Lifetime:NiagaraFloat=&RandomRangeFloat(Minimum=0.5, Maximum=0.6)
    IG_Color:LinearColor=&MakeCustomLinearColorFromBool(Bool=$User.bIsPainted, True Color=$User.PaintedColor, False Color=<rgba 0.3 0.2 0.45 1>)
    IG_Lifetime=0.5
   ) {module=VC_PlayMode}
  ]
  particleSpawnScript [
   &VC_PaintedVariantsOverride(
    Black=<rgba 0.1163 0.1163 0.1163 1>
    BurntSienna=<rgba 0.8177 0.2278 0.0696 1>
    Cobalt=<rgba 0.2066 0.2213 1 1>
    Crimson=<rgba 0.8316 0.091 0.091 1>
    Grey=<rgba 0.2031 0.2031 0.2031 0>
    Lime=<rgba 0.5 1 0 1>
    Orange=<rgba 1 0 0.0372 1>
    Pink=<rgba 0.849 0.1017 0.5772 1>
    Purple=<rgba 0.2789 0 0.6795 0>
    SkyBlue=<rgba 0 1.3308 2 1>
   )
   &InitializeParticle(
    Color Mode=::NewEnumerator0
    Lifetime=&RandomRangeFloat(Minimum=0.5, Maximum=0.6)
    Sprite Rotation Mode=::NewEnumerator2
    Sprite Size Mode=::NewEnumerator1
    Uniform Sprite Size Max=4
    Uniform Sprite Size Min=6
    UsePositionOffset=false
    Write Material Random=true
   )
   &PartitionParticles(Number Of Partitions=$Emitter.PortCount)
   &SkeletalMeshLocation(
    Bone / Socket Index=$Particles.PartitionParticles.Partition
    Bone / Socket Sampling Mode=::NewEnumerator5
    MaskSpawnGroup=false
    Mesh Sampling Type=::NewEnumerator6
    Skeletal Mesh=$System.SkeletalMesh
   )
   &AddVelocity(Velocity=&AddVectorToPosition(Position=$Emitter.Velocity, Vector=&RandomRangeVector(Minimum=<0 100 100>, Maximum=<0 -100 -100>)))
   &ShapeLocation(MaskSpawnGroup=false, Override Seed=false, Sphere Radius=10, UseOwnerScale=true)
   &DLR_ProductPreviews(GameModeID:NiagaraInt32=$System.GameModeID, bPreviewVehicleSpeed:NiagaraBool=$System.bPreviewVehicleSpeed)
  ]
  particleUpdateScript [
   &ParticleState()
   &InheritSourceMovement(
    Applied Position Scale=&Multiply_VectorByFloat(
     Vector=<1 1 1>
     Float=&FloatFromCurve(FloatCurve=DataInterface() {propertyValues {Curve {keys [<at 0 1>,<at 0.7 0>]}}})
    )
    ApplyForce=false
    ApplyPosition=true
    ApplyPositionFalloff=false
   ) {module=InheritSourceMovement001}
   &Drag(
    Drag=&FloatFromCurve(FloatCurve=DataInterface() {propertyValues {Curve {
        defaultValue=0
        keys [
         {arriveTangent=4.1183, interpMode=RCIM_Cubic, leaveTangent=4.1183, tangentMode=RCTM_User, time=0, value=0}
         {arriveTangent=-0.122, interpMode=RCIM_Cubic, leaveTangent=-0.122, tangentMode=RCTM_User, time=0.4067, value=1}
        ]
       }}})
    Use Linear Drag=true
    Use Rotational Drag=true
   )
   &InheritSourceMovement(
    Applied Position Scale=&VectorFromFloat(Value=&FloatFromCurve(FloatCurve=DataInterface() {propertyValues {Curve {
         defaultValue=1
         keys [
          {arriveTangent=-1.7885, interpMode=RCIM_Cubic, leaveTangent=-1.7885, tangentMode=RCTM_User, time=0.05, value=1}
          {interpMode=RCIM_Cubic, tangentMode=RCTM_User, time=1, value=0}
         ]
        }}}))
    ApplyForce=false
    ApplyPosition=true
    ApplyPositionFalloff=false
   )
   &SolveForcesAndVelocity(Clamp Velocity=false, Limit Acceleration=false)
   &ScaleColor(
    Scale Alpha=&FloatFromCurve(FloatCurve=DataInterface() {propertyValues {Curve {keys [{arriveTangent=-1.6042, interpMode=RCIM_Cubic, leaveTangent=-1.6042, tangentMode=RCTM_User, time=0, value=1}, <at 1 0 RCIM_Cubic>]}}})
    ScaleA=true
    ScaleRGB=true
   )
  ]
  rendererClasses [&NiagaraSpriteRendererProperties]
  renderers [{rendererClass=&NiagaraSpriteRendererProperties}]
 }
}
summary {
 systemName=NS_Boost_Spacetear
 emitters [
  {emitterName=NE_VC_Boost_Sprite_Parent, rendererClasses [&NiagaraSpriteRendererProperties]}
  {emitterName=NE_VC_Boost_Ribbon_Parent001, rendererClasses [&NiagaraRibbonRendererProperties]}
 ]
 userVariables [
  User.AlphaMult {type=NiagaraFloat, defaultValue=2, description=""}
  User.bIsFrontendPreview {type=NiagaraBool}
  User.bIsPainted {type=NiagaraBool}
  User.ColorMult {type=NiagaraFloat, defaultValue=1, description=""}
  User.ColorVariantID {type=NiagaraInt32}
  User.DTCarOffset {type=Vector3f, description="", defaultValue {struct=Vector3f, value=<0 0 0>}}
  User.GameModeID {type=NiagaraInt32, defaultValue=3}
  User.PaintedColor {type=LinearColor, description="", defaultValue {struct=LinearColor, value=<rgba 1 1 1 1>}}
  User.PaintedScalar {type=NiagaraFloat, defaultValue=1, description=""}
  User.PortCount {type=NiagaraInt32, defaultValue=1}
  User.Ribbon General Ribbon {
   type=NiagaraDataInterfaceCurve
   description=""
   defaultValue {
    struct=NiagaraExt_VariableValue_DataInterface
    value {
     dataInterface=&self:NiagaraDataInterfaceCurve_0
     dataInterfaceClass=&NiagaraDataInterfaceCurve
    }
   }
  }
  User.Speed Ribbon {type=NiagaraFloat, defaultValue=-2.5, description=""}
 ]
}
