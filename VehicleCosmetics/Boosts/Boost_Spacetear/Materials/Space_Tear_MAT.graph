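# Translucent unlit material for the Spacetear vehicle boost effect, built for Niagara ribbons and showing a tinted starfield-like space inside a torn streak.
# Three scrolling space-texture layers use camera-space view vectors at different scales for fake depth, offset by a normal map.
# Scrolling tear noise, UV fades and a cloud mask shape the alpha, while a panned copy adds edge glow.
# Particle color and dynamic parameters drive tint, emissive, alpha, speed and tiling.
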Material /VehicleCosmetics/Boosts/Boost_Spacetear/Materials/Space_Tear_MAT
graph {
 properties {
  bUsedWithNiagaraRibbons=true
  blendMode=BLEND_Translucent
  pixelDepthOffsetMode=PDOM_Legacy
  refractionMethod=RM_2DOffset
  shadingModel=MSM_Unlit
  usedShadingModels=Unlit
 }
 nodes [
  n1 = TextureCoordinate()
  n2 = Constant2Vector() {r=1}
  n3 = Multiply(A=DynamicParameter() {paramNames [AlphaMult,Emissive,ColorAdd,TextureSpeed]}.TextureSpeed, B=Constant() {r=-2})
  n4 = Add(
   A=Multiply(
    A=MaterialFunctionCall(Speed=@n3) {materialFunction=&TimeWithSpeedVariable}.Time
    B=Constant2Vector() {r=1}
   )
   B=TextureCoordinate()
  )
  n5 = Multiply(A=Power(Base=DotProduct(A=@n1, B=@n2), Exp=Constant() {r=1}), B=@n4)
  Add(A=@n5) {constB=0.5}
  DepthFade() {fadeDistanceDefault=30}
  n6 = TextureCoordinate()
  n7 = Multiply(
   A=Power(Base=ComponentMask(None=@n6) {r=true}, Exp=ScalarParameter() {defaultValue=0.75, parameterName=FadeInMaskPower})
   B=ScalarParameter() {defaultValue=5, parameterName=FadeInMaskMult}
  )
  n8 = TextureSample(UVs=@n4) {texture=/VehicleCosmetics/Boosts/Boost_Spacetear/Textures/T_VCVFX_Boost_SpaceTear_A}
  n9 = Saturate(None=Subtract(A=@n8.RGB, B=Constant() {desc=Internal Cutoff Alpha, r=0.06}))
  n10 = Power(Base=@n8.RGB, Exp=Constant() {desc=Glow Width, r=1})
  n11 = Saturate(None=Multiply(A=Saturate(None=@n7), B=Add(A=@n9, B=@n10)))
  n12 = Multiply(
   A=OneMinus(None=TextureSample() {texture=/DealershipSharedLibrary/FX/Textures/Gradient_Masks/T_VCVFX_boost_Space_Tear_SquareCloud_Mask}.RGB)
   B=@n11
  )
  n13 = ComponentMask(None=TextureCoordinate()) {r=true}
  n14 = Saturate(None=Multiply(A=@n12, B=OneMinus(None=@n13)))
  LinearInterpolate(Alpha=@n14)
  n15 = MaterialFunctionCall(Float3=CameraVectorWS()) {materialFunction=&BreakOutFloat3Components}
  MaterialFunctionCall(X=Multiply(A=@n15.G) {constB=-1}, Y=Multiply(A=@n15.B) {constB=-1}, Z=@n15.R) {materialFunction=&MakeFloat3}
  n16 = MaterialFunctionCall(Speed=Multiply(A=DynamicParameter() {paramNames [AlphaMult,Emissive,ColorAdd,TextureSpeed]}.TextureSpeed)) {materialFunction=&TimeWithSpeedVariable}
  n17 = Transform(None=CameraVectorWS()) {transformSourceType=TRANSFORMSOURCE_World, transformType=TRANSFORM_Camera}
  n18 = Add(
   A=AppendVector(A=@n16.Time, B=Multiply(B=@n16.Time) {constA=-1})
   B=Multiply(A=Constant2Vector() {g=5, r=5}, B=ComponentMask(None=@n17) {g=true, r=true})
  )
  n19 = TextureSample(UVs=@n5) {samplerType=SAMPLERTYPE_Normal, texture=/DealershipSharedLibrary/FX/Textures/Normal/T_VCVFX_boost_SpaceTear_N}
  n20 = Multiply(A=Constant() {r=0.1}, B=ComponentMask(None=Transform(None=@n19.RGB) {transformType=TRANSFORM_Camera}) {g=true, r=true})
  n21 = Multiply(
   A=TextureSample(UVs=Add(A=@n18, B=@n20)) {texture=/VehicleCosmetics/Boosts/Boost_Spacetear/Textures/T_VCVFX_boost_Space_Tear_Space}.RGB
   B=Constant() {r=0.666}
  )
  n22 = MaterialFunctionCall(Speed=Multiply(A=DynamicParameter() {paramNames [AlphaMult,Emissive,ColorAdd,TextureSpeed]}.TextureSpeed, B=Constant() {r=0.6666})) {materialFunction=&TimeWithSpeedVariable}
  n23 = Transform(None=CameraVectorWS()) {transformSourceType=TRANSFORMSOURCE_World, transformType=TRANSFORM_Camera}
  n24 = Add(
   A=AppendVector(A=@n22.Time, B=Multiply(B=@n22.Time) {constA=-1})
   B=Multiply(A=Constant2Vector() {g=2, r=2}, B=ComponentMask(None=@n23) {g=true, r=true})
  )
  n25 = Multiply(
   A=TextureSample(UVs=Add(A=@n24, B=@n20)) {texture=&T_VCVFX_boost_Space_Tear_Space}.RGB
   B=Constant() {r=0.8}
  )
  n26 = MaterialFunctionCall(Speed=Multiply(A=DynamicParameter() {paramNames [AlphaMult,Emissive,ColorAdd,TextureSpeed]}.TextureSpeed, B=Constant() {r=1.1})) {materialFunction=&TimeWithSpeedVariable}
  n27 = AppendVector(A=@n26.Time, B=Multiply(B=@n26.Time) {constA=-1})
  n28 = CameraVectorWS()
  n29 = Multiply(
   A=Constant2Vector() {g=3, r=3}
   B=ComponentMask(None=Transform(None=@n28) {transformSourceType=TRANSFORMSOURCE_World, transformType=TRANSFORM_Camera}) {g=true, r=true}
  )
  n30 = TextureSample(UVs=Add(A=Add(A=@n27, B=@n29), B=@n20)) {texture=&T_VCVFX_boost_Space_Tear_Space}
  n31 = Constant() {r=0.333}
  n32 = Multiply(
   A=Multiply(A=ParticleColor().RGB, B=VectorParameter() {parameterName=Space Color Mult, defaultValue=<rgba 5 5 5 1>}.RGB)
   B=Add(A=Add(A=@n21, B=@n25), B=Multiply(A=@n30.RGB, B=@n31))
  )
  n33 = DynamicParameter() {paramNames [AlphaMult,Emissive,ColorAdd,TextureSpeed]}
  n34 = TextureCoordinate()
  n35 = DynamicParameter() {parameterIndex=1, paramNames [TileX,TileY,ColorM,Param4]}
  n36 = MaterialFunctionCall(X=@n35.TileX, Y=@n35.TileX) {materialFunction=&MakeFloat2}
  n37 = TextureSample(UVs=Panner(Coordinate=Multiply(A=@n34, B=@n36.Result)) {speedX=-3}) {texture=&T_VCVFX_Boost_SpaceTear_A}
  n38 = Constant() {desc=Edge Glow Exp, r=0.5}
  n39 = Subtract(A=@n37.RGB, B=Constant() {desc=Internal Cutoff Alpha, r=0.2})
  n40 = Constant() {desc=Edge Glow Exp, r=0.25}
  n41 = Multiply(A=OneMinus(None=Power(Base=@n39, Exp=@n40)), B=@n35.ColorM) {constB=2.4}
  n42 = ParticleColor()
  n43 = Add(A=Subtract(A=Multiply(A=@n32, B=@n33.Emissive), B=Power(Base=@n37.RGB, Exp=@n38)), B=Multiply(A=@n41, B=@n42.RGB))
  n44 = Multiply(A=@n42.RGB, B=DynamicParameter() {paramNames [AlphaMult,Emissive,ColorAdd,TextureSpeed]}.ColorAdd)
  MaterialFunctionCall(Emissive=Clamp(None=Multiply(A=@n43, B=@n44)) {maxDefault=5}) {materialFunction=&MF_FX_EmissiveAdjustments}
  Multiply(A=@n14) {constB=0}
  Multiply(A=@n37.RGB) {constB=0.25}
  NamedRerouteDeclaration(None=@n19.RGB) {name=Normal, nodeColor=<rgba 0 0 1 1>}
  NamedRerouteUsage()
  Reroute(None=@n19.RGB)
  n45 = Multiply(A=DynamicParameter() {paramNames [AlphaMult,Emissive,ColorAdd,TextureSpeed]}.AlphaMult, B=@n12)
  Saturate(None=StaticSwitchParameter(True=Multiply(A=@n42.A, B=@n45), False=@n45) {parameterName=UseVertexAlpha})
  Saturate(None=@n41)
  n46 = Constant() {r=1}
  n47 = Constant() {r=100}
  StaticSwitchParameter(
   True=Multiply(
    A=@n11
    B=MaterialFunctionCall(Opacity In=@n46, DepthFade Distance=@n47) {materialFunction=&MF_VCVFX_DepthFade}.Result
   )
   False=@n11
  ) {parameterName=UseDepthBiasBlend}
 ]
}
