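# Unlit translucent Niagara sprite material for a glowing star dot in the Space Tear vehicle boost cosmetic effect.
# A circle glow texture's RGB is tinted by particle color and passed through a shared emissive adjustment function.
# Texture alpha times particle alpha is multiplied by a depth fade function to soften intersections with scene geometry.
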Material /VehicleCosmetics/Boosts/Boost_Spacetear/Materials/Space_Tear_StarDot_MAT
graph {
 properties {bUsedWithNiagaraSprites=true, blendMode=BLEND_Translucent, pixelDepthOffsetMode=PDOM_Legacy, shadingModel=MSM_Unlit, usedShadingModels=Unlit}
 nodes [
  n1 = TextureSample() {texture=/VehicleCosmetics/Boosts/Boost_Spacetear/Textures/T_VCVFX_Boost_SpaceTear_CircleGlow}
  n2 = ParticleColor()
  MaterialFunctionCall(Emissive=Multiply(A=@n1.RGB, B=@n2.RGB)) {materialFunction=&MF_FX_EmissiveAdjustments}
  Multiply(
   A=Multiply(A=@n1.A, B=@n2.A)
   B=MaterialFunctionCall(Opacity In=Constant() {r=1}, DepthFade Distance=Constant() {r=30}) {materialFunction=&MF_VCVFX_DepthFade}.Result
  )
 ]
}
