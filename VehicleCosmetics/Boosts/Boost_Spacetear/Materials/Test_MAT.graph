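# Masked, Default Lit test material for Niagara ribbons in the Spacetear boost cosmetic folder.
# A tiled CheckerPattern is multiplied by the rounded, saturated V texture coordinate, a hard step that blacks out half of the UV range.
# The product feeds the Emissive input of MF_FX_EmissiveAdjustments.
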Material /VehicleCosmetics/Boosts/Boost_Spacetear/Materials/Test_MAT
graph {
 properties {
  bAutomaticallySetUsageInEditor=false
  bUsedWithNiagaraRibbons=true
  blendMode=BLEND_Masked
  pixelDepthOffsetMode=PDOM_Legacy
  usedShadingModels=Default Lit
 }
 nodes [
  n1 = MaterialFunctionCall!(UVs=TextureCoordinate(), Tiling=Constant2Vector() {g=2, r=2}) {materialFunction=/Engine/ArtTools/RenderToTexture/MaterialFunctions/CheckerPattern}
  n2 = Round(None=ComponentMask(None=TextureCoordinate()) {g=true})
  MaterialFunctionCall(Emissive=Multiply(A=Multiply(A=@n1.Result), B=Saturate(None=@n2))) {materialFunction=&MF_FX_EmissiveAdjustments}
 ]
}
